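# Foley sound cue for the Solar Theory emote that opens with a start sound and keeps adding whoops.
# A concatenator first plays a mixer layering the quieter start wave with one of four random whoops, then enters a loop of random whoops separated by random delays of a few seconds.
# Emote attenuation, foley sound class and concurrency limits govern playback, which continues when silent and also feeds a music analysis submix.
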SoundCue /BRCosmetics/Sounds/Emotes/SolarTheory/EmoteFoley_SolarTheory_Whoop_Cue
root=Concatenator_0
nodes [
 {id=Concatenator_0, children [Mixer_0,Looping_0], properties {inputVolume [1,1]}}
 {id=Delay_0, children [Random_1], properties {delayMax=8, delayMin=2}}
 {id=Looping_0, children [Delay_0]}
 {id=Mixer_0, children [WavePlayer_14, Random_2], properties {inputVolume [0.6,1]}}
 {
  id=Random_1
  children [WavePlayer_6,WavePlayer_7,WavePlayer_8,WavePlayer_9]
  properties {bRandomizeWithoutReplacement=false, weights [1,1,1,1]}
 }
 {
  id=Random_2
  children [WavePlayer_10, WavePlayer_11, WavePlayer_12, WavePlayer_13]
  properties {bRandomizeWithoutReplacement=false, weights [1,1,1,1]}
 }
 {id=WavePlayer_10, properties {soundWaveAssetPtr=&/Emote_SolarTheory_Whoop_01}}
 {id=WavePlayer_11, properties {soundWaveAssetPtr=&/Emote_SolarTheory_Whoop_02}}
 {id=WavePlayer_12, properties {soundWaveAssetPtr=&/Emote_SolarTheory_Whoop_03}}
 {id=WavePlayer_13, properties {soundWaveAssetPtr=&/Emote_SolarTheory_Whoop_04}}
 {id=WavePlayer_14, properties {soundWaveAssetPtr=&/Emote_SolarTheory_Start}}
 {id=WavePlayer_6, properties {soundWaveAssetPtr=&Emote_SolarTheory_Whoop_01}}
 {id=WavePlayer_7, properties {soundWaveAssetPtr=&Emote_SolarTheory_Whoop_02}}
 {id=WavePlayer_8, properties {soundWaveAssetPtr=&Emote_SolarTheory_Whoop_03}}
 {id=WavePlayer_9, properties {soundWaveAssetPtr=&Emote_SolarTheory_Whoop_04}}
]
properties {
 attenuationSettings=&Fort_Emotes_Athena_Att
 audioPropertiesSheet=/BRCosmetics/Sounds/Settings/PropertySheets/Emotes/APS_Cosmetic_Emote_Foley
 maxDistance=1600
 pitchMultiplier=0.98
 soundClassObject=/Game/Sounds/FortSoundClassesAndModes/Fort_Foley_Emote
 virtualizationMode=PlayWhenSilent
 volumeMultiplier=1
 concurrencyOverrides {maxCount=3}
 concurrencySet [/Game/Sounds/Concurrency_Rules/Athena_EmoteFoley_Conc]
 soundSubmixSends [{minSendLevel=0, soundSubmix=&Fortnite_AudioAnalysis_Music}]
}
